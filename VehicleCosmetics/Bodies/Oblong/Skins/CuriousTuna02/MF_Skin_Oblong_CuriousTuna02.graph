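# Material function for the CuriousTuna02 Oblong vehicle skin: white and pink fresnel-shaded paint, with animated pink glow in the R-masked area.
# A panning alpha-channel starburst is shaped by Perlin, gradient and fresnel noise.
# Twelve mirrored G/B-channel stamps, clipped to single tiles with Floor/Ceil, bob or rock on sine timers.
# Their maxed solid and flicker masks are shared through named reroutes, alongside BaseColor, for compositing into the material attributes.
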MaterialFunction /VehicleCosmetics/Bodies/Oblong/Skins/CuriousTuna02/MF_Skin_Oblong_CuriousTuna02
graph {nodes [
  n1 = Time() {bOverride_Period=true, period=600}
  Add(A=@n1) {constB=0.2}
  n2 = TextureCoordinate() {uTiling=2.75, vTiling=-5.12}
  n3 = ComponentMask(None=@n2) {r=true}
  AppendVector(
   A=ConstantBiasScale(None=ComponentMask(None=@n2) {g=true}) {bias=-0.5, scale=2}
   B=Abs(None=ConstantBiasScale(None=@n3) {bias=-0.5, scale=2})
  )
  Constant()
  n4 = NamedRerouteUsage()
  n5 = TextureSampleParameter2D() {parameterName=RGBMask, texture=&/T_Oblong_CuriousTuna02_RGB}
  n6 = Clamp(None=Multiply(A=@n5.R) {constB=3})
  n7 = Constant3Vector() {desc=Pink, constant=<rgba 1 0.4735 0.6514 1>}
  n8 = TextureCoordinate() {coordinateIndex=1}
  n9 = ComponentMask(None=@n8) {g=true}
  n10 = Multiply(A=@n8) {constB=10}
  n11 = Constant() {r=1}
  n12 = Add(
   A=MaterialFunctionCall!(
    Texture Coordinates=@n8
    Tiling X=@n11
    Tiling Y=Constant() {r=0.5}
    PanSpeedX=Constant()
    PanSpeedY=@n11
    Angle
    RotateSpeed
    Balance
    Chaos=@n11
   ) {materialFunction=/VehicleCosmetics/Materials/MaterialFunctions/MF_Noisemaker_Perlin}.Perlin1
   B=MaterialFunctionCall!(Texture Coordinates=@n8, Tiling=@n11, PanSpeed=Constant() {r=3}, Angle, RotateSpeed, Sharpen, Offset) {materialFunction=/VehicleCosmetics/Materials/MaterialFunctions/MF_Noisemaker_Gradient}.Gradient2
  )
  n13 = Clamp(None=ConstantBiasScale(None=@n9) {bias=-0.5, scale=2})
  n14 = MaterialFunctionCall(Texture Coordinates=@n8, Tiling=@n11, Angle=Constant() {r=90}, Offset=@n11) {materialFunction=&MF_Noisemaker_Gradient}
  n15 = Add(
   A=MaterialFunctionCall!(Fresnel Power=Constant() {r=3}, Normal=Constant3Vector() {constant=<rgba 0 0 1 0>}) {materialFunction=/VehicleCosmetics/Materials/MaterialFunctions/MF_Fresnel_Prime}.Result
   B=Multiply(A=Add(A=@n12, B=@n13), B=@n14.Gradient2)
  )
  n16 = Clamp(None=ConstantBiasScale(None=@n14.Gradient3) {bias=-0.3, scale=3}) {maxDefault=3}
  n17 = Add(A=@n2, B=Constant2Vector() {g=1.75, r=-0.35})
  n18 = TextureSampleParameter2D(UVs=@n17) {parameterName=RGBMask, texture=&T_Oblong_CuriousTuna02_RGB}
  n19 = Floor(None=@n17)
  n20 = Subtract(A=Abs(None=Ceil(None=@n17)), B=Multiply(A=Abs(None=@n19)) {constB=2})
  n21 = Multiply(A=Clamp(None=ComponentMask(None=@n20) {r=true}), B=Clamp(None=ComponentMask(None=@n20) {g=true}))
  n22 = ConstantBiasScale(None=OneMinus(None=Multiply(A=@n18.G, B=@n21) {constB=2})) {bias=-0.75, scale=20}
  n23 = Multiply(
   A=TextureSampleParameter2D(UVs=Panner(Coordinate=@n10) {speedY=0.1}) {parameterName=RGBMask, texture=&T_Oblong_CuriousTuna02_RGB}.A
   B=Multiply(A=Subtract(A=@n15, B=@n16), B=Clamp(None=@n22))
  )
  n24 = Clamp(None=ConstantBiasScale(None=Multiply(A=@n9, B=@n23)) {bias=-0.6, scale=5})
  n25 = LinearInterpolate(A=LinearInterpolate(A=Multiply(A=@n7, B=@n24), Alpha=NamedRerouteUsage()) {constB=0}, B=@n7, Alpha=NamedRerouteUsage()) {constB=0}
  n26 = SetMaterialAttributes!(
   MaterialAttributes
   BaseColor=@n4
   Metallic=ScalarParameter() {defaultValue=1, parameterName=Metallic}
   Roughness=ScalarParameter() {defaultValue=0.5, parameterName=Roughness}
   EmissiveColor=Add(A=Multiply(A=@n4) {constB=0.75}, B=Multiply(A=@n6, B=@n25))
   Normal
   AmbientOcclusion
  ) {attributeSetTypes [
    #1
    #2
    #3
    #4
    #5
    #6
   ]}
  FunctionOutput(None=@n26)
  n27 = TextureCoordinate() {coordinateIndex=1}
  n28 = ComponentMask(None=@n27) {r=true}
  n29 = AppendVector(
   A=ConstantBiasScale(None=ComponentMask(None=@n27) {g=true}) {bias=-0.5, scale=2}
   B=Abs(None=ConstantBiasScale(None=@n28) {bias=-0.5, scale=2})
  )
  n30 = Add(A=Time() {bOverride_Period=true, period=600}) {constB=0.2}
  n31 = Max(A=Sine(None=@n30), B=Sine(None=Add(A=@n30) {constB=0.15}))
  n32 = Multiply(A=@n31) {constB=0.04}
  n33 = Constant2Vector() {g=-1.75, r=1.5}
  n34 = Rotator(Coordinate=Add(A=Multiply(A=@n29) {constB=5}, B=Add(A=@n32, B=@n33)), Time=Constant() {r=2})
  n35 = TextureSampleParameter2D(UVs=@n34) {parameterName=RGBMask, texture=&T_Oblong_CuriousTuna02_RGB}
  n36 = Floor(None=@n34)
  n37 = Subtract(A=Abs(None=Ceil(None=@n34)), B=Multiply(A=Abs(None=@n36)) {constB=2})
  n38 = ComponentMask(None=@n37) {r=true}
  n39 = ComponentMask(None=@n37) {g=true}
  n40 = Multiply(A=Clamp(None=Multiply(A=@n35.G) {constB=2}), B=Multiply(A=Clamp(None=@n38), B=Clamp(None=@n39))) {constB=2}
  n41 = Max(A=Sine(None=@n1), B=Sine(None=Add(A=@n1) {constB=0.15}))
  n42 = AppendVector(A=Constant(), B=Multiply(A=@n41) {constB=0.04})
  n43 = Constant2Vector() {g=-1.2, r=0.5}
  n44 = Rotator(Coordinate=Add(A=Multiply(A=@n29) {constB=3.5}, B=Add(A=@n42, B=@n43)), Time=Constant() {r=-1.25})
  n45 = TextureSampleParameter2D(UVs=@n44) {parameterName=RGBMask, texture=&T_Oblong_CuriousTuna02_RGB}
  n46 = Floor(None=@n44)
  n47 = Subtract(A=Abs(None=Ceil(None=@n44)), B=Multiply(A=Abs(None=@n46)) {constB=2})
  n48 = ComponentMask(None=@n47) {r=true}
  n49 = ComponentMask(None=@n47) {g=true}
  n50 = Multiply(A=Clamp(None=Multiply(A=@n45.G) {constB=2}), B=Multiply(A=Clamp(None=@n48), B=Clamp(None=@n49))) {constB=2}
  n51 = Max(A=@n40, B=@n50)
  n52 = Add(A=Time() {bOverride_Period=true, period=600}) {constB=0.3}
  n53 = Max(A=Sine(None=@n52), B=Sine(None=Add(A=@n52) {constB=0.15}))
  n54 = Multiply(A=@n53) {constB=0.04}
  n55 = AppendVector(A=@n54, B=@n54)
  n56 = Constant2Vector() {g=-1.75, r=3}
  n57 = Rotator(Coordinate=Add(A=Multiply(A=@n29) {constB=6.5}, B=Add(A=@n55, B=@n56)), Time=Constant() {r=4})
  n58 = TextureSampleParameter2D(UVs=@n57) {parameterName=RGBMask, texture=&T_Oblong_CuriousTuna02_RGB}
  n59 = Floor(None=@n57)
  n60 = Subtract(A=Abs(None=Ceil(None=@n57)), B=Multiply(A=Abs(None=@n59)) {constB=2})
  n61 = ComponentMask(None=@n60) {r=true}
  n62 = ComponentMask(None=@n60) {g=true}
  n63 = Multiply(A=Clamp(None=Multiply(A=@n58.G) {constB=2}), B=Multiply(A=Clamp(None=@n61), B=Clamp(None=@n62))) {constB=2}
  n64 = Add(A=Time() {bOverride_Period=true, period=600}) {constB=0.4}
  n65 = Max(A=Sine(None=@n64), B=Sine(None=Add(A=@n64) {constB=0.15}))
  n66 = AppendVector(A=Constant(), B=Multiply(A=@n65) {constB=0.04})
  n67 = Constant2Vector() {g=-3, r=3.25}
  n68 = Rotator(Coordinate=Add(A=Multiply(A=@n29) {constB=7}, B=Add(A=@n66, B=@n67)), Time=Constant() {r=-1})
  n69 = TextureSampleParameter2D(UVs=@n68) {parameterName=RGBMask, texture=&T_Oblong_CuriousTuna02_RGB}
  n70 = Floor(None=@n68)
  n71 = Subtract(A=Abs(None=Ceil(None=@n68)), B=Multiply(A=Abs(None=@n70)) {constB=2})
  n72 = ComponentMask(None=@n71) {r=true}
  n73 = ComponentMask(None=@n71) {g=true}
  n74 = Multiply(A=Clamp(None=Multiply(A=@n69.G) {constB=2}), B=Multiply(A=Clamp(None=@n72), B=Clamp(None=@n73))) {constB=2}
  n75 = Max(A=@n63, B=@n74)
  n76 = TextureCoordinate() {coordinateIndex=1}
  n77 = ComponentMask(None=@n76) {r=true}
  n78 = AppendVector(
   A=ConstantBiasScale(None=ComponentMask(None=@n76) {g=true}) {bias=-0.5, scale=2}
   B=Abs(None=ConstantBiasScale(None=@n77) {bias=-0.5, scale=2})
  )
  n79 = Time() {bOverride_Period=true, period=600}
  n80 = Sine(None=Add(A=Multiply(A=@n79) {constB=0.1}) {constB=0.6})
  n81 = Rotator(Coordinate=Add(A=Multiply(A=@n78) {constB=7}, B=Constant2Vector() {g=-3.25, r=-1.25}), Time=Multiply(A=@n80) {constB=3})
  n82 = TextureSampleParameter2D(UVs=@n81) {parameterName=RGBMask, texture=&T_Oblong_CuriousTuna02_RGB}
  n83 = Floor(None=@n81)
  n84 = Subtract(A=Abs(None=Ceil(None=@n81)), B=Multiply(A=Abs(None=@n83)) {constB=2})
  n85 = ComponentMask(None=@n84) {r=true}
  n86 = ComponentMask(None=@n84) {g=true}
  n87 = Multiply(A=Clamp(None=Multiply(A=@n82.B) {constB=2}), B=Multiply(A=Clamp(None=@n85), B=Clamp(None=@n86))) {constB=2}
  n88 = Time() {bOverride_Period=true, period=600}
  n89 = Sine(None=Add(A=Multiply(A=@n88) {constB=0.2}) {constB=0.6})
  n90 = Rotator(Coordinate=Add(A=Multiply(A=@n78) {constB=4}, B=Constant2Vector() {g=-1, r=0.9}), Time=Multiply(A=@n89) {constB=3})
  n91 = TextureSampleParameter2D(UVs=@n90) {parameterName=RGBMask, texture=&T_Oblong_CuriousTuna02_RGB}
  n92 = Floor(None=@n90)
  n93 = Subtract(A=Abs(None=Ceil(None=@n90)), B=Multiply(A=Abs(None=@n92)) {constB=2})
  n94 = ComponentMask(None=@n93) {r=true}
  n95 = ComponentMask(None=@n93) {g=true}
  n96 = Multiply(A=Clamp(None=Multiply(A=@n91.B) {constB=2}), B=Multiply(A=Clamp(None=@n94), B=Clamp(None=@n95))) {constB=2}
  n97 = Time() {bOverride_Period=true, period=600}
  n98 = Sine(None=Add(A=Multiply(A=@n97) {constB=0.125}) {constB=0.6})
  n99 = Rotator(Coordinate=Add(A=Multiply(A=@n78) {constB=7.5}, B=Constant2Vector() {g=-1.75, r=3}), Time=Multiply(A=@n98) {constB=3})
  n100 = TextureSampleParameter2D(UVs=@n99) {parameterName=RGBMask, texture=&T_Oblong_CuriousTuna02_RGB}
  n101 = Floor(None=@n99)
  n102 = Subtract(A=Abs(None=Ceil(None=@n99)), B=Multiply(A=Abs(None=@n101)) {constB=2})
  n103 = ComponentMask(None=@n102) {r=true}
  n104 = ComponentMask(None=@n102) {g=true}
  n105 = Multiply(A=Clamp(None=Multiply(A=@n100.B) {constB=2}), B=Multiply(A=Clamp(None=@n103), B=Clamp(None=@n104))) {constB=2}
  n106 = Time() {bOverride_Period=true, period=600}
  n107 = Sine(None=Add(A=Multiply(A=@n106) {constB=0.15}) {constB=0.6})
  n108 = Rotator(Coordinate=Add(A=Multiply(A=@n78) {constB=8}, B=Constant2Vector() {g=-3, r=4.25}), Time=Multiply(A=@n107) {constB=3})
  n109 = TextureSampleParameter2D(UVs=@n108) {parameterName=RGBMask, texture=&T_Oblong_CuriousTuna02_RGB}
  n110 = Floor(None=@n108)
  n111 = Subtract(A=Abs(None=Ceil(None=@n108)), B=Multiply(A=Abs(None=@n110)) {constB=2})
  n112 = ComponentMask(None=@n111) {r=true}
  n113 = ComponentMask(None=@n111) {g=true}
  n114 = Multiply(A=Clamp(None=Multiply(A=@n109.B) {constB=2}), B=Multiply(A=Clamp(None=@n112), B=Clamp(None=@n113))) {constB=2}
  n115 = Max(A=Max(A=@n87, B=@n96), B=Max(A=@n105, B=@n114))
  n116 = TextureCoordinate() {coordinateIndex=1}
  n117 = ComponentMask(None=@n116) {r=true}
  n118 = AppendVector(
   A=ConstantBiasScale(None=ComponentMask(None=@n116) {g=true}) {bias=-0.5, scale=2}
   B=Abs(None=ConstantBiasScale(None=@n117) {bias=-0.5, scale=2})
  )
  n119 = Time() {bOverride_Period=true, period=600}
  n120 = Sine(None=Add(A=Multiply(A=@n119) {constB=0.2}) {constB=0.6})
  n121 = Rotator(Coordinate=Add(A=Multiply(A=@n118) {constB=9}, B=Constant2Vector() {g=-3.1, r=-0.8}), Time=Multiply(A=@n120) {constB=3})
  n122 = TextureSampleParameter2D(UVs=@n121) {parameterName=RGBMask, texture=&T_Oblong_CuriousTuna02_RGB}
  n123 = Floor(None=@n121)
  n124 = Subtract(A=Abs(None=Ceil(None=@n121)), B=Multiply(A=Abs(None=@n123)) {constB=2})
  n125 = ComponentMask(None=@n124) {r=true}
  n126 = ComponentMask(None=@n124) {g=true}
  n127 = Multiply(A=Clamp(None=Multiply(A=@n122.B) {constB=2}), B=Multiply(A=Clamp(None=@n125), B=Clamp(None=@n126))) {constB=2}
  n128 = Time() {bOverride_Period=true, period=600}
  n129 = Sine(None=Add(A=Multiply(A=@n128) {constB=0.12}) {constB=0.6})
  n130 = Rotator(Coordinate=Add(A=Multiply(A=@n118) {constB=4}, B=Constant2Vector() {g=-0.9, r=3.25}), Time=Multiply(A=@n129) {constB=3})
  n131 = TextureSampleParameter2D(UVs=@n130) {parameterName=RGBMask, texture=&T_Oblong_CuriousTuna02_RGB}
  n132 = Floor(None=@n130)
  n133 = Subtract(A=Abs(None=Ceil(None=@n130)), B=Multiply(A=Abs(None=@n132)) {constB=2})
  n134 = ComponentMask(None=@n133) {r=true}
  n135 = ComponentMask(None=@n133) {g=true}
  n136 = Multiply(A=Clamp(None=Multiply(A=@n131.B) {constB=2}), B=Multiply(A=Clamp(None=@n134), B=Clamp(None=@n135))) {constB=2}
  n137 = Time() {bOverride_Period=true, period=600}
  n138 = Sine(None=Add(A=Multiply(A=@n137) {constB=0.075}) {constB=0.6})
  n139 = Rotator(Coordinate=Add(A=Multiply(A=@n118) {constB=7.5}, B=Constant2Vector() {g=-2, r=4.25}), Time=Multiply(A=@n138) {constB=3})
  n140 = TextureSampleParameter2D(UVs=@n139) {parameterName=RGBMask, texture=&T_Oblong_CuriousTuna02_RGB}
  n141 = Floor(None=@n139)
  n142 = Subtract(A=Abs(None=Ceil(None=@n139)), B=Multiply(A=Abs(None=@n141)) {constB=2})
  n143 = ComponentMask(None=@n142) {r=true}
  n144 = ComponentMask(None=@n142) {g=true}
  n145 = Multiply(A=Clamp(None=Multiply(A=@n140.B) {constB=2}), B=Multiply(A=Clamp(None=@n143), B=Clamp(None=@n144))) {constB=2}
  n146 = Time() {bOverride_Period=true, period=600}
  n147 = Sine(None=Add(A=Multiply(A=@n146) {constB=0.14}) {constB=0.6})
  n148 = Rotator(Coordinate=Add(A=Multiply(A=@n118) {constB=8}, B=Constant2Vector() {g=-1.5, r=7}), Time=Multiply(A=@n147) {constB=3})
  n149 = TextureSampleParameter2D(UVs=@n148) {parameterName=RGBMask, texture=&T_Oblong_CuriousTuna02_RGB}
  n150 = Floor(None=@n148)
  n151 = Subtract(A=Abs(None=Ceil(None=@n148)), B=Multiply(A=Abs(None=@n150)) {constB=2})
  n152 = ComponentMask(None=@n151) {r=true}
  n153 = ComponentMask(None=@n151) {g=true}
  n154 = Multiply(A=Clamp(None=Multiply(A=@n149.B) {constB=2}), B=Multiply(A=Clamp(None=@n152), B=Clamp(None=@n153))) {constB=2}
  n155 = Max(A=Max(A=@n127, B=@n136), B=Max(A=@n145, B=@n154))
  NamedRerouteDeclaration(None=Max(A=Max(A=@n51, B=@n75), B=Max(A=@n115, B=@n155))) {name=ShapesSolid, nodeColor=<rgba 0.4297 0 1 1>}
  n156 = Max(A=Multiply(A=@n40, B=Clamp(None=@n31) {minDefault=0.05}) {constB=2}, B=Multiply(A=@n50, B=Clamp(None=@n41) {minDefault=0.05}) {constB=2})
  n157 = Max(A=Multiply(A=@n63, B=Clamp(None=@n53) {minDefault=0.05}) {constB=2}, B=Multiply(A=@n74, B=Clamp(None=@n65) {minDefault=0.05}) {constB=2})
  n158 = Abs(None=@n80)
  n159 = Clamp(None=Abs(None=ConstantBiasScale(None=@n158) {bias=-0.5, scale=2})) {minDefault=0.05}
  n160 = Abs(None=@n89)
  n161 = Clamp(None=Abs(None=ConstantBiasScale(None=@n160) {bias=-0.5, scale=2})) {minDefault=0.05}
  n162 = Abs(None=@n98)
  n163 = Clamp(None=Abs(None=ConstantBiasScale(None=@n162) {bias=-0.5, scale=2})) {minDefault=0.05}
  n164 = Abs(None=@n107)
  n165 = Clamp(None=Abs(None=ConstantBiasScale(None=@n164) {bias=-0.5, scale=2})) {minDefault=0.05}
  n166 = Max(
   A=Max(A=Multiply(A=@n87, B=@n159) {constB=2}, B=Multiply(A=@n96, B=@n161) {constB=2})
   B=Max(A=Multiply(A=@n105, B=@n163) {constB=2}, B=Multiply(A=@n114, B=@n165) {constB=2})
  )
  n167 = Abs(None=@n120)
  n168 = Clamp(None=Abs(None=ConstantBiasScale(None=@n167) {bias=-0.5, scale=2})) {minDefault=0.05}
  n169 = Abs(None=@n129)
  n170 = Clamp(None=Abs(None=ConstantBiasScale(None=@n169) {bias=-0.5, scale=2})) {minDefault=0.05}
  n171 = Abs(None=@n138)
  n172 = Clamp(None=Abs(None=ConstantBiasScale(None=@n171) {bias=-0.5, scale=2})) {minDefault=0.05}
  n173 = Abs(None=@n147)
  n174 = Clamp(None=Abs(None=ConstantBiasScale(None=@n173) {bias=-0.5, scale=2})) {minDefault=0.05}
  n175 = Max(
   A=Max(A=Multiply(A=@n127, B=@n168) {constB=2}, B=Multiply(A=@n136, B=@n170) {constB=2})
   B=Max(A=Multiply(A=@n145, B=@n172) {constB=2}, B=Multiply(A=@n154, B=@n174) {constB=2})
  )
  NamedRerouteDeclaration(None=Max(A=Max(A=@n156, B=@n157), B=Max(A=@n166, B=@n175))) {name=ShapesFlicker, nodeColor=<rgba 0 1 0.1094 1>}
  n176 = Constant3Vector() {constant=<rgba 0 0 1 0>}
  n177 = MaterialFunctionCall(Fresnel Power=Constant() {r=3}, Normal=@n176) {materialFunction=&MF_Fresnel_Prime}
  n178 = Power(Base=OneMinus(None=@n177.Result)) {constExponent=20, desc=finishIntFalloff}
  n179 = Power(Base=@n177.Result) {desc=FinishOuterFalloff}
  n180 = Clamp(None=Max(A=Clamp(None=@n178), B=Clamp(None=@n179))) {desc=FinishDarkMinLevel, minDefault=0.15}
  n181 = Constant3Vector() {desc=white color, constant=<rgba 0.9301 0.8632 0.8632 1>}
  n182 = LinearInterpolate(A=Desaturation(None=Multiply(A=@n7) {constB=0.25}, Fraction=Constant() {r=-2}), B=@n7, Alpha=@n5.R)
  n183 = MaterialFunctionCall(Fresnel Power=Constant() {r=3}, Normal=@n176) {materialFunction=&MF_Fresnel_Prime}
  n184 = Clamp(None=Power(Base=OneMinus(None=@n183.Result)) {constExponent=30, desc=MetallicFinishIntFalloff}) {desc=MetallicDarkMinLevel, minDefault=0.4}
  n185 = Power(Base=@n183.Result) {constExponent=3, desc=MetallicFinishOuterFalloff}
  n186 = Multiply(A=Max(A=@n184, B=Clamp(None=@n185))) {constB=0.5}
  n187 = Multiply(A=Power(Base=@n184) {constExponent=60, desc=MetallicFinishIntFalloff}) {constB=0.5, desc=Metallic Highlight}
  n188 = ComponentMask(None=TextureCoordinate() {coordinateIndex=1}) {g=true}
  n189 = Clamp(None=Subtract(A=OneMinus(None=@n188)) {constB=0.25})
  n190 = NamedRerouteUsage()
  n191 = Max(
   A=ConstantBiasScale(None=Multiply(A=@n189, B=@n23) {desc=Starburst Input Here}) {bias=-0.2, scale=10}
   B=Clamp(None=Multiply(A=@n190) {constB=2})
  ) {desc=Shapes Input here}
  n192 = Clamp(None=LinearInterpolate(A=Add(A=@n186, B=@n187), Alpha=Clamp(None=@n191)) {constB=0.1})
  NamedRerouteDeclaration(None=LinearInterpolate(A=Multiply(A=@n180, B=@n181), B=Multiply(A=@n182, B=@n192), Alpha=@n6)) {name=BaseColor, nodeColor=<rgba 1 0 0.2109 1>}
 ]}
